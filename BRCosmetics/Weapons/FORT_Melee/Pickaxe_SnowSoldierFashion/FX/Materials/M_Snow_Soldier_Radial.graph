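# Translucent unlit Niagara sprite material that draws a procedural radial ring effect for the Snow Soldier Fashion pickaxe, without textures.
# SmoothStep bands cut from one centered SphereMask form three concentric rings at stepped intensities, split by a thin vertical gap mask and hard-thresholded.
# A soft mask along V shades the emissive, which takes Particle Color and passes through MF_FX_EmissiveAdjustments.
# Opacity is particle alpha times the combined rings.
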Material /BRCosmetics/Weapons/FORT_Melee/Pickaxe_SnowSoldierFashion/FX/Materials/M_Snow_Soldier_Radial
graph {
 properties {bUsedWithNiagaraSprites=true, blendMode=BLEND_Translucent, pixelDepthOffsetMode=PDOM_Legacy, shadingModel=MSM_Unlit, usedShadingModels=Unlit}
 nodes [
  n1 = SphereMask(A=TextureCoordinate() {uTiling=0}, B=Constant2Vector() {g=0.5}, Radius=Constant() {r=0.6}, Hardness=Constant())
  n2 = SphereMask(A=TextureCoordinate(), B=Constant() {r=0.5}, Radius=Constant() {r=0.5}, Hardness=Constant() {r=0.5})
  n3 = SphereMask(A=TextureCoordinate() {vTiling=0}, B=Constant2Vector() {r=0.5}, Radius=Constant() {r=0.05}, Hardness=Constant())
  n4 = Subtract(A=Subtract(A=SmoothStep(Value=@n2) {constMax=0.1}, B=SmoothStep(Value=@n2) {constMax=0.2, constMin=0.15}), B=@n3)
  n5 = SmoothStep(Value=@n2) {constMax=0.25, constMin=0.15}
  n6 = SmoothStep(Value=@n2) {constMax=0.35, constMin=0.3}
  n7 = SmoothStep(Value=Subtract(A=Subtract(A=@n5, B=@n6), B=@n3)) {constMax=0.9, constMin=0.9}
  n8 = Subtract(A=Subtract(A=SmoothStep(Value=@n2) {constMax=0.37, constMin=0.37}, B=SmoothStep(Value=@n2) {constMax=0.45, constMin=0.45}), B=@n3)
  n9 = Add(
   A=Add(A=SmoothStep(Value=@n4) {constMax=0.9, constMin=0.9}, B=Multiply(A=@n7) {constB=0.5})
   B=Multiply(A=SmoothStep(Value=@n8) {constMax=0.9, constMin=0.9}) {constB=0.25}
  )
  n10 = ParticleColor()
  MaterialFunctionCall(Emissive=Multiply(A=Multiply(A=@n1, B=@n9), B=@n10.RGB)) {materialFunction=&MF_FX_EmissiveAdjustments}
  Multiply(A=@n10.A, B=@n9)
 ]
}
