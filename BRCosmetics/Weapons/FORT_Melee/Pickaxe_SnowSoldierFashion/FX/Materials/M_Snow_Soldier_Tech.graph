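# Translucent unlit Niagara sprite material for the SnowSoldierFashion pickaxe FX, drawing a moving band through a triangle-break grayscale texture.
# Dynamic parameter R sweeps a threshold, and two offset Distance tests are hard-stepped and subtracted to isolate a narrow range of texture values.
# The sharpened mask multiplies particle color through MF_FX_EmissiveAdjustments for emissive and particle alpha for opacity.
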Material /BRCosmetics/Weapons/FORT_Melee/Pickaxe_SnowSoldierFashion/FX/Materials/M_Snow_Soldier_Tech
graph {
 properties {
  bOutputTranslucentVelocity=true
  bUsedWithNiagaraSprites=true
  blendMode=BLEND_Translucent
  opacityMaskClipValue=0.02
  pixelDepthOffsetMode=PDOM_Legacy
  shadingModel=MSM_Unlit
  usedShadingModels=Unlit
 }
 nodes [
  n1 = Constant() {r=0.1}
  n2 = TextureSample() {
   samplerType=SAMPLERTYPE_LinearGrayscale
   texture=/BRCosmetics/Weapons/FORT_Melee/Pickaxe_SnowSoldierFashion/FX/Texture/T_Snow_Soldier_TriBreak_M
  }
  n3 = LinearInterpolate(Alpha=DynamicParameter().R) {constA=-0.1, constB=0.8}
  n4 = Distance(A=@n2.R, B=@n3)
  n5 = Distance(A=@n2.R, B=Add(A=@n3) {constB=0.1})
  n6 = SmoothStep(Value=Subtract(
    A=SmoothStep(Min=@n1, Max=@n1, Value=@n4) {constMax=0.5, constMin=0.5}
    B=SmoothStep(Min=@n1, Max=@n1, Value=@n5) {constMax=0.5, constMin=0.5}
   )) {constMin=0.9}
  n7 = ParticleColor()
  MaterialFunctionCall(Emissive=Multiply(A=@n6, B=@n7.RGB)) {materialFunction=&MF_FX_EmissiveAdjustments}
  Multiply(A=@n6, B=@n7.A)
 ]
}
